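; Homing thrown-shield projectile for the Ashton Chicago back pack item, which damages pawns, chains between enemies and returns to its thrower.
; On spawn it hides the held shield, homes on the aimed enemy and starts a return timer.
; Hits run a sphere overlap plus line-of-sight check, then timeline-travel to the nearest visible enemy, otherwise bounce or home back.
; Destruction sends the catch event and restores shield visibility; explosions drive impact FX.
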(event Custom|RotatePrj
  ;[1 1 1] Level out shield on throw
  (bind (_direction _newtrack_0) (Timeline|Timeline_0 Play)
    (:Update
      (Transformation|SetRelativeRotation (Variables|Default|GetSkeletalMesh) (Math|Rotator|Lerp(Rotator) (Variables|/Script/Engine.SceneComponent|GetRelativeRotation (Variables|Default|GetSkeletalMesh)) "0, 0,90.000000" _newtrack_0 true) false false))
    (:Finished
      (Utilities|Time|ClearandInvalidateTimerbyHandle (Variables|Default|GetProjectileRotationTimer)))))

(event EventBeginPlay
  ;[1 1 1] Begin Play.  Return if nothing is hit after a delay
  (|CallParentFunction)
  (Utilities|FlowControl|Sequence
    (:then_0
      (bind _asfort_player_pawn (Utilities|Casting|CastToFortPlayerPawn (Variables|Actor|GetInstigator)))
      (bind _output_get (Variables|Default|SetOwningPlayerPawn _asfort_player_pawn))
      (GameplayEffects|ApplyGameplayEffectToTarget :self (Ability|GetAbilitySystemComponent _output_get) :GameplayEffectClass "/Game/Athena/Items/Gameplay/BackPacks/Ashton/Chicago/GE_Ashton_Chicago_DeepDishDelivery_GE.GE_Ashton_Chicago_DeepDishDelivery_GE_C" :Target (Ability|GetAbilitySystemComponent _output_get) :Level 0.0) ; hides shield on arm mesh.  Needs to go here so it never gets applied until proj has spawned
      (bind _asda_ashton_pack_chicago (Utilities|Casting|CastToDA_AshtonPack_Chicago (Variables|/Script/FortniteGame.FortPawn|GetCurrentWeapon (Variables|Default|GetOwningPlayerPawn))))
      (Development|SetHiddeninGame (Variables|/Script/FortniteGame.FortWeapon|GetWeaponMesh _asda_ashton_pack_chicago) true true) ; Weapon is unhidden in the Catch ability
    )
    (:then_1
      (bind _returnvalue (Utilities|Time|SetTimerbyEvent (EventDispatchers|CreateEvent :!event Custom|RotatePrj) 0.1 true))
      (Variables|Default|SetProjectileRotationTimer _returnvalue))
    (:then_2
      (Utilities|Array|AddUnique (Variables|Default|GetActorsToIgnoreForTargeting) (Variables|Default|GetOwningPlayerPawn))
      (Variables|Default|SetPrjCurveTimer)
      (Utilities|Time|SetTimerbyEvent (delegate Custom|TriggerReturn) (Variables|Default|GetTravelOutTime)))
    (:then_3
      (CallFunction|SetHomingVars))))

(event Custom|TriggerReturn
  ;[1 1 1] Begin Play.  Return if nothing is hit after a delay
  (Variables|Default|SetbReturnTimerExpired true))

(event Collision|EventHit
  (params (MyComp :type Primitive%20Component%20Object%20Reference) (Other :type Actor%20Object%20Reference) (OtherComp :type Primitive%20Component%20Object%20Reference) (bSelfMoved :type Boolean) (HitLocation :type Vector :default "0, 0, 0") (HitNormal :type Vector :default "0, 0, 0") (NormalImpulse :type Vector :default "0, 0, 0") (Hit :type Hit%20Result%20Structure%20%28by%20ref%29))
  (bind (_bblockinghit _binitialoverlap _time _distance _location _impactpoint _normal _impactnormal _physmat _hitactor) (Collision|BreakHitResult Hit))
  ;[1 1 1] Event Hit: Trigger a ShieldSphereTrace.
  (|CallParentFunction MyComp Other OtherComp bSelfMoved HitLocation HitNormal NormalImpulse Hit) ; TODO// consider increasing travel duration with each bounce
  (Variables|Default|SetShieldHitNormal HitNormal)
  (Variables|Default|SetShieldHitLoc HitLocation)
  (Projectile|SetHomingTarget 0 self 0 0 :!class /Script/FortniteGame.FortProjectileMovementInterface)
  (bind _output_get (Variables|Default|SetHitActor _hitactor))
  (Delivery|DoExplosion)
  (bind _asplayer_pawn_athena (Utilities|Casting|CastToPlayerPawn_Athena _output_get)
    (:then
      (Utilities|Array|AddUnique (Variables|Default|GetActorsToIgnoreForTargeting) _asplayer_pawn_athena)
      (Variables|Default|SetbTargetAcquired true)
      (CallFunction|PlayHitPawnSounds _asplayer_pawn_athena))
    (:CastFailed
      (Audio|PlaySoundatLocation (Variables|Default|GetHitEnvSoundCue) (Variables|Default|GetShieldHitLoc))))
  (CallFunction|ShieldSphereTrace))

(event Custom|ShieldSphereTrace
  (bind _owningplayerpawn (Variables|Default|GetOwningPlayerPawn))
  (bind _returnvalue (Transformation|GetActorLocation _owningplayerpawn))
  (bind _shieldhitloc (Variables|Default|GetShieldHitLoc))
  (bind _returnvalue_2 (- _shieldhitloc _returnvalue))
  (bind _returnvalue_3 (Math|Vector|Normalize _returnvalue_2))
  ;[1 1 1] Shield Sphere Trace
  (Variables|Default|SetPawnToTravelTo 0)
  (Variables|Default|SetDistanceToNearestTarget 1e+04)
  (bind _outactors (Collision|SphereOverlapActors (Transformation|GetActorLocation) (ScalableFloat|GetValueatLevel (Variables|Default|GetDataHitSphereTraceRadius)) (Utilities|Array|MakeArray "ObjectTypeQuery3") "/Script/FortniteGame.FortPlayerPawn" (Variables|Default|GetActorsToIgnoreForTargeting)))
  (for _array_element _outactors
    (bind _asfort_player_pawn (Utilities|Casting|CastToFortPlayerPawn _array_element))
    (if (Team|OnSameTeam _asfort_player_pawn (Variables|Default|GetOwningPlayerPawn))
      (Utilities|Array|AddUnique (Variables|Default|GetActorsToIgnoreForTargeting) _asfort_player_pawn) ; We found a nearby teammate. Ignore him for targeting and LOS checks.
      (else
        (bind _clearlostotarget (CallFunction|LOSCheck _asfort_player_pawn)) ; find out if you can see them
        (if _clearlostotarget
          (if (< (Transformation|GetDistanceTo _asfort_player_pawn) (Variables|Default|GetDistanceToNearestTarget))
            (Variables|Default|SetPawnToTravelTo _asfort_player_pawn) ; We have a new closest pawn.
            (Variables|Default|SetDistanceToNearestTarget (Transformation|GetDistanceTo _asfort_player_pawn)) ; We have a new closest distance here.
          )))))
  (bind _output_get (Variables|Default|SetShieldTarget (Variables|Default|GetPawnToTravelTo)))
  (Utilities|IsValid _output_get
    (:Is%20Valid
      (Variables|Default|SetbTargetAcquired true)
      (Collision|SetActorEnableCollision false)
      (CallFunction|Travel) ; Got targets, now we seek them
    )
    (:Is%20Not%20Valid
      (bind _output_get_2 (Variables|Default|SetVector_AngleOfIncidence (Math|Vector|MirrorVectorbyNormal (Variables|/Script/Engine.MovementComponent|GetVelocity (Variables|Movement|GetProjectileMovementComponent)) (Variables|Default|GetShieldHitNormal))))
      (Development|PrintString (Utilities|String|Append "Dot product of hit: " (Utilities|String|ToString(Float) (Math|Vector|DotProduct _returnvalue_3 (Math|Vector|Normalize _output_get_2)))) (Variables|Default|GetbDebug) (Variables|Default|GetbDebug))
      (if (< (Math|Vector|DotProduct _returnvalue_3 (Math|Vector|Normalize _output_get_2)) -0.65)
        _
        (elif (or (Variables|Default|GetbReturnTimerExpired) (Variables|Default|GetbTargetAcquired))
          _
          (else
            (Projectile|ResumeSimulation) ; Bounce if we haven't hit anyone and timer hasn't expired
            (Variables|/Script/Engine.MovementComponent|SetVelocity (Variables|Default|GetVector_AngleOfIncidence) (Variables|Movement|GetProjectileMovementComponent)))))
      (Projectile|ResumeSimulation)
      (CallFunction|PrjReturn) ; Return if there's no one valid and our timer has expired
    )))

(event Custom|Travel
  ;[1 1 1] Travel Between Targets
  (bind (_direction _newtrack_0) (Timeline|TravelToTargetTimeline Play)
    (:Update
      (Utilities|IsValid (Variables|Default|GetShieldTarget)
        (:Is%20Valid
          (Transformation|SetWorldLocation (Variables|Collision|GetCapsuleComponent) (Math|Vector|Lerp(Vector) (Transformation|GetWorldLocation (Variables|Collision|GetCapsuleComponent)) (Transformation|GetActorLocation (Variables|Default|GetShieldTarget)) _newtrack_0) false false)
          (if (<= (Math|Vector|VectorLength (- (Transformation|GetWorldLocation (Variables|Collision|GetCapsuleComponent)) (Transformation|GetActorLocation (Variables|Default|GetShieldTarget)))) (Variables|Default|GetEndTargetProximityCheck))
            (Delivery|DoExplosion)
            (CallFunction|PlayHitPawnSounds (Utilities|Casting|CastToFortPlayerPawn (Variables|Default|GetShieldTarget)))
            (bind _output_get (Variables|Default|SetHitActor (Variables|Default|GetShieldTarget)))
            (CallFunction|EndTravel)))
        (:Is%20Not%20Valid
          (CallFunction|EndTravel))))))

(event Custom|Stop%20Timeline
  ;[1 1 1] Travel Between Targets
  (Timeline|TravelToTargetTimeline Stop))

(event Custom|EndTravel
  ;[1 1 1] Travel Between Targets
  (CallFunction|StopTimeline)
  (Utilities|Array|AddUnique (Variables|Default|GetActorsToIgnoreForTargeting) (stale Custom|Travel _output_get))
  (CallFunction|ShieldSphereTrace))

(event Custom|PrjReturn
  (bind _self self)
  (bind _owningplayerpawn (Variables|Default|GetOwningPlayerPawn))
  ;[0.9202 1 0.9012] Return event - bring shield back
  (Utilities|FlowControl|DoOnce false
    (:Completed
      (bind _output_get (Variables|/Script/Engine.ProjectileMovementComponent|SetMaxSpeed (Math|Float|Clamp(Float) (/ (Math|Vector|VectorLength (- (Transformation|GetActorLocation) (Transformation|GetActorLocation (Variables|Default|GetOwningPlayerPawn)))) (Variables|Default|GetDesiredReturnDuration)) (Variables|Default|GetMoveSpeed_Min) (Variables|Default|GetMoveSpeed_Max)) (Variables|Movement|GetProjectileMovementComponent)))
      (Development|PrintString (Utilities|String|Append "Setting return speed: " (Utilities|String|ToString(Float) _output_get)) (Variables|Default|GetbDebug) (Variables|Default|GetbDebug))
      (Collision|SetActorEnableCollision false)
      (Networking|SwitchHasAuthority
        (:Authority
          (Utilities|IsValid _owningplayerpawn
            (:Is%20Valid
              (Projectile|SetHomingTarget 0 _self _owningplayerpawn 0 :!class /Script/FortniteGame.FortProjectileMovementInterface)
              (Projectile|SetHomingTimeUntilMaxTurnSpeed 0 _self 0.0 :!class /Script/FortniteGame.FortProjectileMovementInterface)
              (Projectile|SetHomingTurnSpeedMin 0 _self 1e+04 :!class /Script/FortniteGame.FortProjectileMovementInterface)
              (Projectile|SetHomingTurnSpeedMax 0 _self 1e+04 :!class /Script/FortniteGame.FortProjectileMovementInterface)
              (Projectile|SetHomingLockTargetDistanceThreshold 0 _self (Variables|Default|GetEndReturnProximityCheck) :!class /Script/FortniteGame.FortProjectileMovementInterface)
              (Projectile|SetHomingTargetOffset _self (Variables|Default|GetReturnHomingOffset) :!class /Script/FortniteGame.FortProjectileMovementInterface))))))))

(event ProjectileMovement|EventOnReachedLockTargetDistanceThreshold
  ;[0.9987 1 0.9767] Trigger catch event
  (Actor|DestroyActor))

(event EventDestroyed
  (bind _owningplayerpawn (Variables|Default|GetOwningPlayerPawn))
  (bind _returnvalue (Ability|GetAbilitySystemComponent _owningplayerpawn))
  (bind _owningplayerpawn_2 (Variables|Default|GetOwningPlayerPawn))
  ;[1 1 1] Event Destroyed: Remove Activation Block GE
  (Tags|UnBlockAbilitiesWithTags (Utilities|Casting|CastToFortAbilitySystemComponent (Ability|GetAbilitySystemComponent (Variables|Default|GetOwningPlayerPawn))) (Variables|Default|GetChangeEquipmentTag) false)
  (Utilities|IsValid _owningplayerpawn_2
    (:Is%20Valid
      (Ability|SendGameplayEventtoActor _owningplayerpawn_2 (Variables|Default|GetCatchEventTag) (Variables|Default|GetCatchEventPayload))
      (CallFunction|UnhideMesh)
      (Utilities|IsValid _returnvalue
        (:Is%20Valid
          (GameplayEffects|RemoveActiveEffectswithGrantedTags _returnvalue "(GameplayTags=((TagName=\"Ability.Ashton.Chicago.DeepDish.Out\")))") ; Shield Visibility
        )))))

(event Custom|UnhideMesh
  (bind _owningplayerpawn (Variables|Default|GetOwningPlayerPawn))
  (Utilities|IsValid _owningplayerpawn
    (:Is%20Valid
      (bind _asda_ashton_pack_chicago (Utilities|Casting|CastToDA_AshtonPack_Chicago (Variables|/Script/FortniteGame.FortPawn|GetCurrentWeapon _owningplayerpawn)))
      (Development|SetHiddeninGame (Variables|/Script/FortniteGame.FortWeapon|GetWeaponMesh _asda_ashton_pack_chicago) false true) ; Weapon is unhidden in the Catch ability
    )))

(event EventOnExploded
  (params (HitActors :type Array%20of%20Actor%20Object%20References) (HitResults :type Array%20of%20Hit%20Result%20Structures))
  (bind _returnvalue (Transformation|GetActorLocation))
  (bind _returnvalue_2 (Transformation|GetActorLocation))
  (bind _p_ashton_chicago_shieldspin (Variables|Default|GetP_Ashton_Chicago_ShieldSpin))
  (bind _p_ashton_chicago_shieldtrail (Variables|Default|GetP_Ashton_Chicago_ShieldTrail))
  (bind _bdebug (Variables|Default|GetbDebug))
  (bind _shieldhitloc (Variables|Default|GetShieldHitLoc))
  ;[0.6069 0.6201 1] Shield Emissive FX while dealing damgae
  (|CallParentFunction HitActors HitResults)
  (Networking|CanExecuteCosmeticEvents
    (:True
      (Utilities|FlowControl|Sequence
        (:then_0
          (Effects|Components|ParticleSystem|SetFloatParameter _p_ashton_chicago_shieldspin "Emissive" 1e+01)
          (Utilities|FlowControl|Delay 0.1)
          (Effects|Components|ParticleSystem|SetFloatParameter _p_ashton_chicago_shieldspin "Emissive" 0.0))
        (:then_1
          (Effects|Components|ParticleSystem|SetVectorParameter _p_ashton_chicago_shieldtrail "TrailColor" (Math|Conversions|ToVector(LinearColor) (Variables|VFX|GetFX_TrailEmisColor)))
          (Utilities|FlowControl|Delay 0.15)
          (Effects|Components|ParticleSystem|SetVectorParameter _p_ashton_chicago_shieldtrail "TrailColor" "1.000000,1.000000,1.000000"))
        (:then_2
          (Utilities|Casting|CastToFortPlayerPawn (Variables|Default|GetHitActor)
            (:then
              (bind _shieldtarget (Variables|Default|GetShieldTarget)
                (:then
                  (Effects|SpawnEmitteratLocation "/Game/Athena/Items/Gameplay/BackPacks/Ashton/Chicago/FX_Assets/P_Ashton_Chicago_ShieldImpact_Directional.P_Ashton_Chicago_ShieldImpact_Directional" _returnvalue (Math|Rotator|FindLookatRotation (Transformation|GetWorldLocation (Variables|Collision|GetCapsuleComponent)) (Transformation|GetActorLocation _shieldtarget))) ; over Two enemy targets
                  ;;[1 1 1] Debug
                  (Development|PrintString "MultiTarget" _bdebug _bdebug "(R=1.000000,G=0.000000,B=0.040244,A=1.000000)")
                  (if _bdebug
                    (Rendering|Debug|DrawDebugSphere _returnvalue 5e+01 12 "(R=1.000000,G=0.016900,B=0.000000,A=1.000000)" 3.0 1.0)))
                (:else
                  (Effects|SpawnEmitteratLocation "/Game/Athena/Items/Gameplay/BackPacks/Ashton/Chicago/FX_Assets/P_Ashton_Chicago_ShieldImpact_Radial.P_Ashton_Chicago_ShieldImpact_Radial" _returnvalue_2 "90.000000, 0, 0")
                  (Development|PrintString "HitWorld" _bdebug _bdebug "(R=0.000000,G=0.055177,B=1.000000,A=1.000000)")
                  (if _bdebug
                    (Rendering|Debug|DrawDebugSphere _returnvalue_2 5e+01 12 "(R=0.000000,G=0.224197,B=1.000000,A=1.000000)" 3.0 1.0)))) ; Play Radial FX if we only hit one pawn.  Directional if there are multiple
            )
            (:CastFailed
              (Effects|SpawnEmitteratLocation "/Game/Athena/Items/Gameplay/BackPacks/Ashton/Chicago/FX_Assets/P_Ashton_Chicago_ShieldImpact_Radial.P_Ashton_Chicago_ShieldImpact_Radial" _shieldhitloc "90.000000, 0, 0")
              (Development|PrintString "OneHitThenReturn" _bdebug _bdebug "(R=0.000000,G=1.000000,B=0.117374,A=1.000000)")
              (if _bdebug
                (Rendering|Debug|DrawDebugSphere _shieldhitloc 5e+01 12 "(R=0.127447,G=1.000000,B=0.000000,A=1.000000)" 3.0 1.0)))) ; Play Radial effect if we dont hit a pawn
        )))))

(event OnComponentActivated(ProjectileMovementComponent)
  (params (Component :type Actor%20Component%20Object%20Reference) (bReset :type Boolean))
  ;[0.6069 0.6201 1] Shield Emissive FX while dealing damgae
  (Networking|CanExecuteCosmeticEvents
    (:True
      (Effects|Particles|Trails|BeginTrails (Variables|Default|GetP_Ashton_Chicago_ShieldTrail) "ArmStrap_A" "ArmStrap_B" "ETrailWidthMode_FromCentre" 5.0))))

(fn LOSCheck
  (params (TracedActor :type Actor%20Object%20Reference))
  (bind _outhit (Collision|LineTraceByChannel (Transformation|GetActorLocation) (Transformation|GetActorLocation TracedActor) "TraceTypeQuery1" false (Variables|Default|GetActorsToIgnoreForTargeting)))
  (if (Utilities|Equal(Object) (.HitActor (Collision|BreakHitResult _outhit)) TracedActor)
    (return true)
    (else
      (return false))))

(fn OnHit
  (params (HitLocation :type Vector) (HitNormal :type Vector))
  (Variables|/Script/Engine.MovementComponent|SetVelocity (Math|Vector|MirrorVectorbyNormal (Variables|/Script/Engine.MovementComponent|GetVelocity (Variables|Movement|GetProjectileMovementComponent)) HitNormal) (Variables|Movement|GetProjectileMovementComponent)))

(fn OnRep_bHit)

(fn OnRep_bReturnTimerExpired
  (if (Variables|Default|GetbReturnTimerExpired)
    (if (not (Variables|Default|GetbTargetAcquired))
      (CallFunction|PrjReturn))))

(fn OnRep_HitLocation)

(fn OnRep_PlayerLocation)

(fn OnRep_Target)

(fn PlayHitPawnSounds
  (params (HitPawn :type Fort%20Player%20Pawn%20Object%20Reference))
  (bind _owningplayerpawn (Variables|Default|GetOwningPlayerPawn))
  (Utilities|IsValid _owningplayerpawn
    (:Is%20Valid
      (if (Pawn|IsLocallyControlled _owningplayerpawn)
        (Audio|PlaySound2D (Variables|Default|GetHitPlayerSoundCue))
        (elif (Pawn|IsLocallyControlled HitPawn)
          (Audio|PlaySound2D (Variables|Default|GetHitPlayerSoundCue))
          (else
            (Audio|PlaySoundatLocation (Variables|Default|GetHitEnvSoundCue) (Transformation|GetActorLocation))))))))

(fn PrintActorsFromArray
  (params (ActorArray :type Array%20of%20Actor%20Object%20References))
  (for _ ActorArray
    (Development|PrintString (Utilities|String|Append "Hit;" (Utilities|GetDisplayName (Utilities|Array|Get(acopy) ActorArray (LocalVariables|GetHitActorIndex)))) true true "(R=1.000000,G=0.019947,B=0.059748,A=1.000000)" 5.0)
    (LocalVariables|SetHitActorIndex (+ (LocalVariables|GetHitActorIndex) 1))))

(fn SetHomingVars
  (bind _self self)
  (bind _homingtarget (Variables|Default|GetHomingTarget))
  (bind _owningplayerpawn (Variables|Default|GetOwningPlayerPawn))
  (bind _returnvalue (Fortnite|GetAimedAtEnemy (Fortnite|GetFortPlayerControllerfromActor _owningplayerpawn) (ScalableFloat|GetValueatLevel (Variables|Default|GetDataHomingRectHeightandWidth)) (ScalableFloat|GetValueatLevel (Variables|Default|GetDataHomingRectHeightandWidth)) (ScalableFloat|GetValueatLevel (Variables|Default|GetDataHomingRange))))
  (Projectile|SetHomingStyle 0 0 "LockOn" false :!class /Script/FortniteGame.FortProjectileMovementInterface)
  (bind _asfort_player_pawn (Utilities|Casting|CastToFortPlayerPawn _returnvalue))
  (bind _output_get (Variables|Default|SetHomingTarget _asfort_player_pawn))
  (Utilities|Array|AddUnique (Variables|Default|GetActorsToIgnoreForTargeting) _output_get)
  (Networking|SwitchHasAuthority
    (:Authority
      (Projectile|SetHomingTurnSpeedMin 0 _self 1e+04 :!class /Script/FortniteGame.FortProjectileMovementInterface)
      (Projectile|SetHomingTurnSpeedMax 0 _self 1e+04 :!class /Script/FortniteGame.FortProjectileMovementInterface)
      (Projectile|SetHomingTimeUntilMaxTurnSpeed 0 _self 0.0 :!class /Script/FortniteGame.FortProjectileMovementInterface)
      (Projectile|SetHomingTargetOffset _self (Variables|Default|GetTargetHomingOffset) :!class /Script/FortniteGame.FortProjectileMovementInterface)
      (Utilities|IsValid _homingtarget
        (:Is%20Valid
          (if (>= (Math|Vector|DotProduct (Math|Vector|Normalize (- (Transformation|GetActorLocation _returnvalue) (Transformation|GetActorLocation _owningplayerpawn))) (Math|Vector|Normalize (Transformation|GetActorForwardVector _owningplayerpawn))) 0.49)
            (Projectile|SetHomingTarget 0 self _homingtarget 0 :!class /Script/FortniteGame.FortProjectileMovementInterface)))))))

(fn ConstructionScript
  (|CallParentFunction))
